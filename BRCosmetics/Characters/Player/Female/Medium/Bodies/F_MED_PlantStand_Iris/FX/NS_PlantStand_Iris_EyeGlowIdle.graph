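# Niagara system for the idle eye glow on the PlantStand Iris player body cosmetic.
# A single burst spawns two persistent sprites, each using its partition index to follow a skeletal mesh bone or socket, nudged toward the camera.
# User.showFX drives glow alpha, with a dimmer MobileColor derived alongside and two sprite renderers.
# An alternative fade based on the dot product of two sampled bone orientations is present but disabled.
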NiagaraSystem /BRCosmetics/Characters/Player/Female/Medium/Bodies/F_MED_PlantStand_Iris/FX/NS_PlantStand_Iris_EyeGlowIdle
dependencies {
 usedDataInterfaces [&NiagaraDataInterfaceSkeletalMesh]
 usedDynamicInputs []
 usedModules [
  &self:SystemSpawnScript.SystemScriptSource.SystemScriptGraph.NiagaraNodeAssignment_1.SetVariables_#1
  &SystemState
 ]
}
emitters {EyeGlows {
  emitterUpdateScript [
   &EmitterState(
    Inactive Response=::NewEnumerator1
    Life Cycle Mode=::NewEnumerator1
    Loop Behavior=::NewEnumerator1
    Loop Duration Mode=::NewEnumerator1
    UseLoopDelay=false
   )
   &SpawnBurst_Instantaneous(Spawn Count=2, Use Spawn Probability=false, UseLoopCountLimit=true)
  ]
  particleSpawnScript [
   &InitializeParticle(Color=<rgba 0.8 1 0.9333 0>, Sprite Size Mode=::NewEnumerator3, Uniform Sprite Size=12, UsePositionOffset=false, Write Material Random=true)
   &PartitionParticles()
   &CameraOffset(Camera Offset Amount=3)
  ]
  particleUpdateScript [
   &ParticleState(Kill Particles When Lifetime Has Elapsed=false, Let Infinitely Lived Particles Die When Emitter Deactivates=true)
   &SkeletalMeshLocation(
    Bone / Socket Index=$Particles.PartitionParticles.Partition
    Bone Sampling Mode=::NewEnumerator4
    MaskSpawnGroup=false
    Sampled Position Offset=<0 0 -0.5>
    Skeletal Mesh=$System.Skeletal Mesh
   )
   &SampleSkeletalMesh(
    AdditionalYawPitchRoll=false
    Bone / Socket Index=4
    Bone Sampling Mode:ENiagaraBoneSamplingMode=::NewEnumerator4
    Position Offset Coordinate Space:ENiagaraCoordinateSpace=::Local
    Sampled Orientation Coordinate Space:ENiagaraCoordinateSpace=::Local
    Skeletal Mesh:NiagaraDataInterfaceSkeletalMesh=$System.Skeletal Mesh
   ) {enabled=false}
   &SampleSkeletalMesh(
    AdditionalYawPitchRoll=false
    Bone / Socket Index=3
    Bone Sampling Mode:ENiagaraBoneSamplingMode=::NewEnumerator4
    Position Offset Coordinate Space:ENiagaraCoordinateSpace=::Local
    Sampled Orientation Coordinate Space:ENiagaraCoordinateSpace=::Local
    Skeletal Mesh:NiagaraDataInterfaceSkeletalMesh=$System.Skeletal Mesh
   ) {enabled=false, module=SampleSkeletalMesh001}
   /SetVariables(Particles.Color=&MultiplyLinearColor(
     Linear Color=<rgba 0.8 1 0.9333 1>
     Scale Factor=&MakeFloat4FromVectorAndFloat(
      "Vector (RGB)"=<1 1 1>
      "Float (Alpha)"=&Multiply_Float(
       A=$User.showFX
       B=&RemapRange(
        Input Value To Remap=&DotProduct(
         Vector A=&ReturnMeshOrientationAxis(Mesh Orientation Axis=::ZAxis, Mesh Orientation Quat=$Particles.SampleSkeletalMesh.SampledRotation)
         Vector B=&ReturnMeshOrientationAxis(Mesh Orientation Axis=::ZAxis, Mesh Orientation Quat=$Particles.SampleSkeletalMesh001.SampledRotation)
        )
        Input Min=0.3
        Input Max=0.6
        Output Min=0
        Output Max=1
       )
      )
     )
    )) {enabled=false, isSetParameters=true}
   /SetVariables(Particles.Color=&MultiplyLinearColor(
     Linear Color=<rgba 0.8 1 0.9333 1>
     Scale Factor=&MakeFloat4FromVectorAndFloat("Vector (RGB)"=<1 1 1>, "Float (Alpha)"=$User.showFX)
    )) {isSetParameters=true}
   /SetVariables(Particles.MobileColor:LinearColor=&MultiplyLinearColor(Linear Color=$Particles.Color, Scale Factor=<1 1 1 0.3>)) {isSetParameters=true}
  ]
  rendererClasses [&NiagaraSpriteRendererProperties]
  renderers [
   {rendererClass=&NiagaraSpriteRendererProperties}
   {rendererClass=&NiagaraSpriteRendererProperties, rendererIndex=1}
  ]
 }}
summary {
 systemName=NS_PlantStand_Iris_EyeGlowIdle
 emitters [{emitterName=EyeGlows, rendererClasses [&NiagaraSpriteRendererProperties]}]
 userVariables [
  User.bIsFrontEnd {type=NiagaraBool}
  User.bIsFrontEndPreview {type=NiagaraBool}
  User.bKillReaction {type=NiagaraBool}
  User.global Alpha {type=NiagaraFloat, defaultValue=0, description=""}
  User.IsSkydiving {type=NiagaraBool}
  User.showFX {type=NiagaraFloat, defaultValue=1, description=""}
 ]
}
