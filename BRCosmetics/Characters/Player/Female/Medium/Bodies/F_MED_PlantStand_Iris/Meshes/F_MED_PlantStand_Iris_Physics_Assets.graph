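# Physics asset for the F_MED_PlantStand_Iris female body, set up for secondary motion of skirt panels and hip bandages.
# Capsules on pelvis, thighs and calves form the body volumes beside the dynamic cloth-bone chains.
# Chains start at small "mstr" bodies, several constrained to the pelvis, with angular-limited links down each chain.
# Free-angular cross constraints tie the back centre chain to its side chains and each front chain to its fr2 counterpart.
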PhysicsAsset /BRCosmetics/Characters/Player/Female/Medium/Bodies/F_MED_PlantStand_Iris/Meshes/F_MED_PlantStand_Iris_Physics_Assets
bodies {
 calf_l [{shapeType=Capsule, length=26, radius=12.5, shapeName=calf_l_capsule, center=<-23 -2 0>, rotation=<rot 90 0 0>}]
 calf_r [{shapeType=Capsule, length=25.7767, radius=11.8851, shapeName=calf_r_capsule, center=<22.58 1.5885 0>, rotation=<rot 90 0 0>}]
 dyn_hip_bandage_left_01 [{shapeType=Capsule, length=5, radius=6, shapeName=dyn_skirt_fr_l_02_capsule, center=<6.5 0 4>, rotation=<rot 0 0 90>}]
 dyn_hip_bandage_left_02 [{shapeType=Capsule, length=5, radius=6, shapeName=dyn_skirt_fr_l_02_capsule, center=<6.5 0 4>, rotation=<rot 0 0 90>}]
 dyn_hip_bandage_left_mstr [{shapeType=Capsule, length=0.505, radius=0.505, shapeName=dyn_hip_bandage_left_mstr_capsule, rotation=<rot 0 -90 0>}]
 dyn_hip_bandage_right_01 [{shapeType=Capsule, length=5, radius=6, shapeName=dyn_skirt_fr_l_02_capsule, center=<6.5 0 4>, rotation=<rot 0 0 90>}]
 dyn_hip_bandage_right_02 [{shapeType=Capsule, length=5, radius=6, shapeName=dyn_skirt_fr_l_02_capsule, center=<6.5 0 4>, rotation=<rot 0 0 90>}]
 dyn_hip_bandage_right_mstr [{shapeType=Capsule, length=0.505, radius=0.505, shapeName=dyn_hip_bandage_right_mstr_capsule, rotation=<rot 0 -90 0>}]
 dyn_skirt_01_l [{shapeType=Capsule, length=11, radius=5.5, shapeName=dyn_skirt_r_01_capsule, center=<6.5 0 3>, rotation=<rot 0 0 90>}]
 dyn_skirt_01_l_mstr [{shapeType=Sphere, radius=1, shapeName=dyn_skirt_01_l_mstr_sphere}]
 dyn_skirt_01_r [{shapeType=Capsule, length=11, radius=5.5, shapeName=dyn_skirt_r_01_capsule, center=<6.5 0 4>, rotation=<rot 0 0 90>}]
 dyn_skirt_01_r_mstr [{shapeType=Sphere, radius=1, shapeName=dyn_skirt_01_r_mstr_sphere}]
 dyn_skirt_02_l [{shapeType=Capsule, length=11, radius=5.5, shapeName=dyn_skirt_r_02_capsule, center=<6.5 0 3>, rotation=<rot 0 0 90>}]
 dyn_skirt_02_r [{shapeType=Capsule, length=11, radius=5.5, shapeName=dyn_skirt_r_02_capsule, center=<6.5 0 4>, rotation=<rot 0 0 90>}]
 dyn_skirt_bk_01 [{shapeType=Capsule, length=17.5, radius=6, shapeName=dyn_skirt_bk_01_capsule, center=<6.5 0 4>, rotation=<rot 0 0 90>}]
 dyn_skirt_bk_01_l [{shapeType=Capsule, length=15.5, radius=5.5, shapeName=dyn_skirt_bk_l_01_capsule, center=<6.5 0 3.5>, rotation=<rot 0 0 90>}]
 dyn_skirt_bk_01_l_mstr [{shapeType=Sphere, radius=1, shapeName=dyn_skirt_bk_01_l_mstr_sphere}]
 dyn_skirt_bk_01_mstr [{shapeType=Sphere, radius=1, shapeName=dyn_skirt_bk_01_mstr_sphere}]
 dyn_skirt_bk_01_r [{shapeType=Capsule, length=15.5, radius=5.5, shapeName=dyn_skirt_bk_r_01_capsule, center=<6.5 0 3.5>, rotation=<rot 0 0 90>}]
 dyn_skirt_bk_01_r_mstr [{shapeType=Sphere, radius=1, shapeName=dyn_skirt_bk_01_r_mstr_sphere}]
 dyn_skirt_bk_02 [{shapeType=Capsule, length=19, radius=6.5, shapeName=dyn_skirt_bk_02_capsule, center=<6.5 0 2>, rotation=<rot 0 0 90>}]
 dyn_skirt_bk_02_l [{shapeType=Capsule, length=19, radius=6.5, shapeName=dyn_skirt_bk_l_02_capsule, center=<6.5 0 3>, rotation=<rot 0 0 90>}]
 dyn_skirt_bk_02_r [{shapeType=Capsule, length=19, radius=6.5, shapeName=dyn_skirt_bk_r_02_capsule, center=<6.5 0 3.5>, rotation=<rot 0 0 90>}]
 dyn_skirt_bk_03 [{shapeType=Capsule, length=24.5, radius=7, shapeName=dyn_skirt_bk_03_capsule, center=<6.5 0 2.5>, rotation=<rot 0 0 90>}]
 dyn_skirt_bk_03_l [{shapeType=Capsule, length=19, radius=6.5, shapeName=dyn_skirt_bk_l_03_capsule, center=<6.5 0 3>, rotation=<rot 0 0 90>}]
 dyn_skirt_bk_03_r [{shapeType=Capsule, length=19, radius=6.5, shapeName=dyn_skirt_bk_r_03_capsule, center=<6.5 0 3.5>, rotation=<rot 0 0 90>}]
 dyn_skirt_fr2_01_l [{shapeType=Capsule, length=16, radius=6, shapeName=dyn_skirt_fr_l_02_capsule, center=<6.5 0 5>, rotation=<rot 0 0 90>}]
 dyn_skirt_fr2_01_l_mstr [{shapeType=Capsule, length=0.505, radius=0.505, shapeName=dyn_skirt_fr2_01_l_mstr_capsule, rotation=<rot 0 -90 0>}]
 dyn_skirt_fr2_01_r [{shapeType=Capsule, length=14, radius=5.5, shapeName=dyn_skirt_fr_r_01_capsule, center=<6.5 0 4>, rotation=<rot 0 0 90>}]
 dyn_skirt_fr2_01_r_mstr [{shapeType=Sphere, radius=1, shapeName=dyn_skirt_fr2_01_r_mstr_sphere}]
 dyn_skirt_fr2_02_l [{shapeType=Capsule, length=16, radius=6, shapeName=dyn_skirt_fr_l_02_capsule, center=<6.5 0 5.5>, rotation=<rot 0 0 90>}]
 dyn_skirt_fr2_02_r [{shapeType=Capsule, length=16, radius=6, shapeName=dyn_skirt_fr_r_02_capsule, center=<6.5 0 5.5>, rotation=<rot 0 0 90>}]
 dyn_skirt_fr2_03_l [{shapeType=Capsule, length=16, radius=6, shapeName=dyn_skirt_fr_l_02_capsule, center=<6.5 0 5.5>, rotation=<rot 0 0 90>}]
 dyn_skirt_fr2_03_r [{shapeType=Capsule, length=16, radius=6, shapeName=dyn_skirt_fr_r_02_capsule, center=<6.5 0 5.5>, rotation=<rot 0 0 90>}]
 dyn_skirt_fr_01_l [{shapeType=Capsule, length=5, radius=6, shapeName=dyn_skirt_fr_r_02_capsule, center=<5 0 4.5>, rotation=<rot 0 0 90>}]
 dyn_skirt_fr_01_l_mstr [{shapeType=Sphere, radius=1, shapeName=dyn_skirt_fr_01_l_mstr_sphere}]
 dyn_skirt_fr_01_r [{shapeType=Capsule, length=5, radius=6, shapeName=dyn_skirt_fr_r_02_capsule, center=<5 0 4.5>, rotation=<rot 0 0 90>}]
 dyn_skirt_fr_01_r_mstr [{shapeType=Capsule, length=0.505, radius=0.505, shapeName=dyn_skirt_fr_01_r_mstr_capsule, rotation=<rot 0 -90 0>}]
 dyn_skirt_fr_02_l [{shapeType=Capsule, length=5, radius=6, shapeName=dyn_skirt_fr_r_02_capsule, center=<6.5 0 4>, rotation=<rot 0 0 90>}]
 dyn_skirt_fr_02_r [{shapeType=Capsule, length=5, radius=6, shapeName=dyn_skirt_fr_r_02_capsule, center=<6.5 0 4>, rotation=<rot 0 0 90>}]
 dyn_skirt_fr_03_l [{shapeType=Capsule, length=5, radius=6, shapeName=dyn_skirt_fr_r_02_capsule, center=<6.5 0 3>, rotation=<rot 0 0 90>}]
 dyn_skirt_fr_03_r [{shapeType=Capsule, length=5, radius=6, shapeName=dyn_skirt_fr_r_02_capsule, center=<6.5 0 3.5>, rotation=<rot 0 0 90>}]
 pelvis [{shapeType=Capsule, length=9, radius=10, center=<-3.5 -1 0>}, {shapeType=Capsule, length=9, radius=11, center=<-18.2567 -1 0>}]
 thigh_l [{shapeType=Capsule, length=30, radius=11, shapeName=thigh_l_capsule, center=<-25 -1.5 1>, rotation=<rot 90 0 3>}]
 thigh_r [{shapeType=Capsule, length=30, radius=11, shapeName=thigh_r_capsule, center=<22 1.5 -1>, rotation=<rot 90 0 3>}]
}
constraints [
 {
  bone1Name=dyn_skirt_01_l_mstr
  bone2Name=pelvis
  swing1LimitDegrees=45
  swing1Motion=Free
  swing2LimitDegrees=45
  swing2Motion=Free
  twistLimitDegrees=45
  twistMotion=Free
 }
 {
  bone1Name=dyn_skirt_01_r_mstr
  bone2Name=pelvis
  swing1LimitDegrees=45
  swing1Motion=Free
  swing2LimitDegrees=45
  swing2Motion=Free
  twistLimitDegrees=45
  twistMotion=Free
 }
 {
  bone1Name=dyn_skirt_bk_01_l_mstr
  bone2Name=pelvis
  swing1LimitDegrees=45
  swing1Motion=Free
  swing2LimitDegrees=45
  swing2Motion=Free
  twistLimitDegrees=45
  twistMotion=Free
 }
 {
  bone1Name=dyn_skirt_bk_01_mstr
  bone2Name=pelvis
  swing1LimitDegrees=45
  swing1Motion=Free
  swing2LimitDegrees=45
  swing2Motion=Free
  twistLimitDegrees=45
  twistMotion=Free
 }
 {bone1Name=dyn_skirt_bk_01, bone2Name=dyn_skirt_bk_01_mstr, swing1LimitDegrees=30, swing2LimitDegrees=30, twistLimitDegrees=5}
 {bone1Name=dyn_skirt_bk_02, bone2Name=dyn_skirt_bk_01, swing1LimitDegrees=30, swing2LimitDegrees=30, twistLimitDegrees=5}
 {bone1Name=dyn_skirt_bk_03, bone2Name=dyn_skirt_bk_02, swing1LimitDegrees=30, swing2LimitDegrees=30, twistLimitDegrees=5}
 {
  bone1Name=dyn_skirt_bk_01_r_mstr
  bone2Name=pelvis
  swing1LimitDegrees=45
  swing1Motion=Free
  swing2LimitDegrees=45
  swing2Motion=Free
  twistLimitDegrees=45
  twistMotion=Free
 }
 {
  bone1Name=dyn_skirt_fr2_01_r_mstr
  bone2Name=pelvis
  swing1LimitDegrees=45
  swing1Motion=Free
  swing2LimitDegrees=45
  swing2Motion=Free
  twistLimitDegrees=45
  twistMotion=Free
 }
 {
  bone1Name=dyn_skirt_fr_01_l_mstr
  bone2Name=pelvis
  swing1LimitDegrees=45
  swing1Motion=Free
  swing2LimitDegrees=45
  swing2Motion=Free
  twistLimitDegrees=45
  twistMotion=Free
 }
 {
  bone1Name=calf_l
  bone2Name=thigh_l
  swing1LimitDegrees=45
  swing1Motion=Free
  swing2LimitDegrees=45
  swing2Motion=Free
  twistLimitDegrees=45
  twistMotion=Free
 }
 {
  bone1Name=calf_r
  bone2Name=thigh_r
  swing1LimitDegrees=45
  swing1Motion=Free
  swing2LimitDegrees=45
  swing2Motion=Free
  twistLimitDegrees=45
  twistMotion=Free
 }
 {bone1Name=dyn_skirt_01_l, bone2Name=dyn_skirt_01_l_mstr, swing1LimitDegrees=30, swing2LimitDegrees=20, twistLimitDegrees=5}
 {bone1Name=dyn_skirt_02_l, bone2Name=dyn_skirt_01_l, swing1LimitDegrees=30, swing2LimitDegrees=20, twistLimitDegrees=5}
 {bone1Name=dyn_skirt_01_r, bone2Name=dyn_skirt_01_r_mstr, swing1LimitDegrees=30, swing2LimitDegrees=20, twistLimitDegrees=5}
 {bone1Name=dyn_skirt_02_r, bone2Name=dyn_skirt_01_r, swing1LimitDegrees=30, swing2LimitDegrees=20, twistLimitDegrees=5}
 {bone1Name=dyn_skirt_bk_01_l, bone2Name=dyn_skirt_bk_01_l_mstr, swing1LimitDegrees=30, swing2LimitDegrees=30, twistLimitDegrees=5}
 {bone1Name=dyn_skirt_bk_02_l, bone2Name=dyn_skirt_bk_01_l, swing1LimitDegrees=30, swing2LimitDegrees=30, twistLimitDegrees=5}
 {bone1Name=dyn_skirt_bk_03_l, bone2Name=dyn_skirt_bk_02_l, swing1LimitDegrees=30, swing2LimitDegrees=30, twistLimitDegrees=5}
 {bone1Name=dyn_skirt_bk_01_r, bone2Name=dyn_skirt_bk_01_r_mstr, swing1LimitDegrees=30, swing2LimitDegrees=30, twistLimitDegrees=5}
 {bone1Name=dyn_skirt_bk_02_r, bone2Name=dyn_skirt_bk_01_r, swing1LimitDegrees=30, swing2LimitDegrees=30, twistLimitDegrees=5}
 {bone1Name=dyn_skirt_bk_03_r, bone2Name=dyn_skirt_bk_02_r, swing1LimitDegrees=30, swing2LimitDegrees=30, twistLimitDegrees=5}
 {bone1Name=dyn_skirt_fr2_01_r, bone2Name=dyn_skirt_fr2_01_r_mstr, swing1LimitDegrees=30, swing2LimitDegrees=10, twistLimitDegrees=5}
 {bone1Name=dyn_skirt_fr2_02_r, bone2Name=dyn_skirt_fr2_01_r, swing1LimitDegrees=30, swing2LimitDegrees=10, twistLimitDegrees=5}
 {bone1Name=dyn_skirt_fr2_03_r, bone2Name=dyn_skirt_fr2_02_r, swing1LimitDegrees=30, swing2LimitDegrees=10, twistLimitDegrees=5}
 {bone1Name=dyn_skirt_fr_01_r, bone2Name=dyn_skirt_fr_01_r_mstr, swing1LimitDegrees=30, swing2LimitDegrees=10, twistLimitDegrees=5}
 {bone1Name=dyn_skirt_fr_02_r, bone2Name=dyn_skirt_fr_01_r, swing1LimitDegrees=30, swing2LimitDegrees=10, twistLimitDegrees=5}
 {bone1Name=dyn_skirt_fr_03_r, bone2Name=dyn_skirt_fr_02_r, swing1LimitDegrees=30, swing2LimitDegrees=10, twistLimitDegrees=5}
 {bone1Name=dyn_hip_bandage_right_02, bone2Name=dyn_hip_bandage_right_01, swing1LimitDegrees=10, swing2LimitDegrees=10, twistLimitDegrees=5}
 {bone1Name=dyn_hip_bandage_right_01, bone2Name=dyn_hip_bandage_right_mstr, swing1LimitDegrees=10, swing2LimitDegrees=10, twistLimitDegrees=5}
 {bone1Name=dyn_skirt_fr2_03_l, bone2Name=dyn_skirt_fr2_02_l, swing1LimitDegrees=30, swing2LimitDegrees=15, twistLimitDegrees=5}
 {bone1Name=dyn_skirt_fr2_02_l, bone2Name=dyn_skirt_fr2_01_l, swing1LimitDegrees=30, swing2LimitDegrees=15, twistLimitDegrees=5}
 {bone1Name=dyn_skirt_fr2_01_l, bone2Name=dyn_skirt_fr2_01_l_mstr, swing1LimitDegrees=30, swing2LimitDegrees=15, twistLimitDegrees=5}
 {bone1Name=dyn_hip_bandage_left_02, bone2Name=dyn_hip_bandage_left_01, swing1LimitDegrees=30, swing2LimitDegrees=30, twistLimitDegrees=5}
 {bone1Name=dyn_hip_bandage_left_01, bone2Name=dyn_hip_bandage_left_mstr, swing1LimitDegrees=30, swing2LimitDegrees=30, twistLimitDegrees=5}
 {
  bone1Name=thigh_l
  bone2Name=pelvis
  swing1LimitDegrees=45
  swing1Motion=Free
  swing2LimitDegrees=45
  swing2Motion=Free
  twistLimitDegrees=45
  twistMotion=Free
 }
 {
  bone1Name=thigh_r
  bone2Name=pelvis
  swing1LimitDegrees=45
  swing1Motion=Free
  swing2LimitDegrees=45
  swing2Motion=Free
  twistLimitDegrees=45
  twistMotion=Free
 }
 {bone1Name=dyn_skirt_fr_01_l, bone2Name=dyn_skirt_fr_01_l_mstr, swing1LimitDegrees=30, swing2LimitDegrees=15, twistLimitDegrees=5}
 {bone1Name=dyn_skirt_fr_02_l, bone2Name=dyn_skirt_fr_01_l, swing1LimitDegrees=30, swing2LimitDegrees=15, twistLimitDegrees=5}
 {bone1Name=dyn_skirt_fr_03_l, bone2Name=dyn_skirt_fr_02_l, swing1LimitDegrees=30, swing2LimitDegrees=15, twistLimitDegrees=5}
 {
  bone1Name=dyn_skirt_fr_01_r
  bone2Name=dyn_skirt_fr2_01_r
  swing1LimitDegrees=45
  swing1Motion=Free
  swing2LimitDegrees=45
  swing2Motion=Free
  twistLimitDegrees=45
  twistMotion=Free
 }
 {
  bone1Name=dyn_skirt_fr_02_r
  bone2Name=dyn_skirt_fr2_02_r
  swing1LimitDegrees=45
  swing1Motion=Free
  swing2LimitDegrees=45
  swing2Motion=Free
  twistLimitDegrees=45
  twistMotion=Free
 }
 {
  bone1Name=dyn_skirt_fr_01_l
  bone2Name=dyn_skirt_fr2_01_l
  swing1LimitDegrees=45
  swing1Motion=Free
  swing2LimitDegrees=45
  swing2Motion=Free
  twistLimitDegrees=45
  twistMotion=Free
 }
 {
  bone1Name=dyn_skirt_fr_02_l
  bone2Name=dyn_skirt_fr2_02_l
  swing1LimitDegrees=45
  swing1Motion=Free
  swing2LimitDegrees=45
  swing2Motion=Free
  twistLimitDegrees=45
  twistMotion=Free
 }
 {
  bone1Name=dyn_skirt_bk_01
  bone2Name=dyn_skirt_bk_01_l
  swing1LimitDegrees=45
  swing1Motion=Free
  swing2LimitDegrees=45
  swing2Motion=Free
  twistLimitDegrees=45
  twistMotion=Free
 }
 {
  bone1Name=dyn_skirt_bk_02
  bone2Name=dyn_skirt_bk_02_l
  swing1LimitDegrees=45
  swing1Motion=Free
  swing2LimitDegrees=45
  swing2Motion=Free
  twistLimitDegrees=45
  twistMotion=Free
 }
 {
  bone1Name=dyn_skirt_bk_03
  bone2Name=dyn_skirt_bk_03_l
  swing1LimitDegrees=45
  swing1Motion=Free
  swing2LimitDegrees=45
  swing2Motion=Free
  twistLimitDegrees=45
  twistMotion=Free
 }
 {
  bone1Name=dyn_skirt_bk_01
  bone2Name=dyn_skirt_bk_01_r
  swing1LimitDegrees=45
  swing1Motion=Free
  swing2LimitDegrees=45
  swing2Motion=Free
  twistLimitDegrees=45
  twistMotion=Free
 }
 {
  bone1Name=dyn_skirt_bk_02
  bone2Name=dyn_skirt_bk_02_r
  swing1LimitDegrees=45
  swing1Motion=Free
  swing2LimitDegrees=45
  swing2Motion=Free
  twistLimitDegrees=45
  twistMotion=Free
 }
 {
  bone1Name=dyn_skirt_bk_03
  bone2Name=dyn_skirt_bk_03_r
  swing1LimitDegrees=45
  swing1Motion=Free
  swing2LimitDegrees=45
  swing2Motion=Free
  twistLimitDegrees=45
  twistMotion=Free
 }
 {
  bone1Name=dyn_skirt_fr_03_l
  bone2Name=dyn_skirt_fr2_03_l
  swing1LimitDegrees=45
  swing1Motion=Free
  swing2LimitDegrees=45
  swing2Motion=Free
  twistLimitDegrees=45
  twistMotion=Free
 }
 {
  bone1Name=dyn_skirt_fr_03_r
  bone2Name=dyn_skirt_fr2_03_r
  swing1LimitDegrees=45
  swing1Motion=Free
  swing2LimitDegrees=45
  swing2Motion=Free
  twistLimitDegrees=45
  twistMotion=Free
 }
]
