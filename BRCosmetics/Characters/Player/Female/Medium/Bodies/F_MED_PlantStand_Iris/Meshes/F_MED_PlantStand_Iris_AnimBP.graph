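# Post-process style animation Blueprint for the F_MED_PlantStand_Iris cosmetic body, adding secondary motion to its skirt and hip bandages.
# The copied parent pose gets skirt constraints and a Control Rig, then is cached.
# bRigidBodyEnabled selects RigidBody plus AnimDynamics, or movement-weighted trail controllers; both constrain bandages to thighs, and two dance emotes add bone corrections afterwards.
# The event graph computes movement and state alphas and feeds WPO cloth-wind material parameters.
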AnimBlueprint /BRCosmetics/Characters/Player/Female/Medium/Bodies/F_MED_PlantStand_Iris/Meshes/F_MED_PlantStand_Iris_AnimBP
target_skeleton=&/F_MED_PlantStand_Iris_Skeleton
graphs [
 AnimGraph {nodes [
   n1 = RigidBody!(ComponentPose=Local To Component(LocalPose=Use cached pose 'PreCache'()), Alpha, OverrideWorldGravity:Vector="0,0,-980", ExternalForce) -> (Pose) {
    bClampLinearTranslationLimitToRefPose=true
    bOverrideWorldGravity=true
    overridePhysicsAsset=&/F_MED_PlantStand_Iris_Physics_Assets
    simulationSpace=BaseBoneSpace
    alphaBoolBlend {blendOption=Linear}
    baseBoneRef {boneName=pelvis}
    componentAppliedLinearAccClamp=<3000 3000 3000>
    componentLinearAccScale=<0.5 0.1 0.3>
    componentLinearVelScale=<0.25 0.1 0.3>
    simSpaceSettings {
     maxAngularAcceleration=1000
     maxAngularVelocity=3
     maxLinearAcceleration=700
     maxLinearVelocity=500
     velocityScaleZ=5
     worldAlpha=1
     externalLinearDragV=<0.2 0.2>
    }
   }
   n2 = "Constraint - dyn_hip_bandage_left_mstr (P)"(ComponentPose=@n1, Alpha=1, ConstraintWeights_0=1) {
    alphaBoolBlend {blendOption=Linear}
    boneToModify {boneName=dyn_hip_bandage_left_mstr}
    constraintSetup [{targetBone {boneName=thigh_l}}]
    constraintWeights [1]
   }
   n3 = "Constraint - dyn_hip_bandage_center_left_mstr (PP)"!(ComponentPose=@n2, Alpha=1, ConstraintWeights_0=0.7, ConstraintWeights_1:float=0.3) -> (Pose) {
    alphaBoolBlend {blendOption=Linear}
    boneToModify {boneName=dyn_hip_bandage_center_left_mstr}
    constraintSetup [{targetBone {boneName=thigh_l}}, {targetBone {boneName=thigh_r}}]
    constraintWeights [1,1]
   }
   n4 = "Constraint - dyn_hip_bandage_right_mstr (P)"(ComponentPose=@n3, Alpha=1, ConstraintWeights_0=1) {
    alphaBoolBlend {blendOption=Linear}
    boneToModify {boneName=dyn_hip_bandage_right_mstr}
    constraintSetup [{targetBone {boneName=thigh_r}}]
    constraintWeights [1]
   }
   n5 = "Constraint - dyn_hip_bandage_center_right_mstr (PP)"!(ComponentPose=@n4, Alpha=1, ConstraintWeights_0=0.7, ConstraintWeights_1:float=0.3) -> (Pose) {
    alphaBoolBlend {blendOption=Linear}
    boneToModify {boneName=dyn_hip_bandage_center_right_mstr}
    constraintSetup [{targetBone {boneName=thigh_r}}, {targetBone {boneName=thigh_l}}]
    constraintWeights [1,1]
   }
   n6 = Get Not Jumping Alpha() -> (Not Jumping Alpha:double)
   n7 = "Transform (Modify) Bone - Bone: dyn_hip_bandage_left_mstr"(ComponentPose=@n5, Alpha=@n6, Translation="0,0,-3", Rotation="0,0,0", Scale="1,1,1") {
    rotationMode=BMM_Replace
    rotationSpace=BCS_BoneSpace
    scaleMode=BMM_Replace
    scaleSpace=BCS_BoneSpace
    translationMode=BMM_Replace
    translationSpace=BCS_BoneSpace
    alphaBoolBlend {blendOption=Linear}
    alphaScaleBiasClamp {bInterpResult=true}
    boneToModify {boneName=dyn_hip_bandage_left_mstr}
    translation {z=-3}
   }
   n8 = "Transform (Modify) Bone - Bone: dyn_hip_bandage_right_mstr"(ComponentPose=@n7, Alpha=@n6, Translation="0,0,-1", Rotation="0,0,0", Scale="1,1,1") {
    rotationMode=BMM_Replace
    rotationSpace=BCS_BoneSpace
    scaleMode=BMM_Replace
    scaleSpace=BCS_BoneSpace
    translationMode=BMM_Replace
    translationSpace=BCS_BoneSpace
    alphaBoolBlend {blendOption=Linear}
    alphaScaleBiasClamp {bInterpResult=true}
    boneToModify {boneName=dyn_hip_bandage_right_mstr}
    translation {z=-1}
   }
   n9 = "Transform (Modify) Bone - Bone: dyn_hip_bandage_center_left_mstr"(ComponentPose=@n8, Alpha=@n6, Translation="0,0,-3", Rotation="-15,0,0", Scale="1,1,1") {
    rotationMode=BMM_Replace
    rotationSpace=BCS_BoneSpace
    scaleMode=BMM_Replace
    scaleSpace=BCS_BoneSpace
    translationMode=BMM_Replace
    translationSpace=BCS_BoneSpace
    alphaBoolBlend {blendOption=Linear}
    alphaScaleBiasClamp {bInterpResult=true}
    boneToModify {boneName=dyn_hip_bandage_center_left_mstr}
    rotation {pitch=-15}
    translation {z=-3}
   }
   n10 = "Transform (Modify) Bone - Bone: dyn_hip_bandage_center_right_mstr"(ComponentPose=@n9, Alpha=@n6, Translation="0,0,-1", Rotation="-10,0,0", Scale="1,1,1") {
    rotationMode=BMM_Replace
    rotationSpace=BCS_BoneSpace
    scaleMode=BMM_Replace
    scaleSpace=BCS_BoneSpace
    translationMode=BMM_Replace
    translationSpace=BCS_BoneSpace
    alphaBoolBlend {blendOption=Linear}
    alphaScaleBiasClamp {bInterpResult=true}
    boneToModify {boneName=dyn_hip_bandage_center_right_mstr}
    rotation {pitch=-10}
    translation {z=-1}
   }
   n11 = Get Jumping Alpha() -> (Jumping Alpha:double)
   n12 = "Transform (Modify) Bone - Bone: dyn_hip_bandage_center_left_mstr"(ComponentPose=@n10, Alpha=@n11, Translation="0,0,2", Rotation="0,0,0", Scale="1,1,1") {
    rotationMode=BMM_Replace
    rotationSpace=BCS_BoneSpace
    scaleMode=BMM_Replace
    scaleSpace=BCS_BoneSpace
    translationMode=BMM_Replace
    translationSpace=BCS_BoneSpace
    alphaBoolBlend {blendOption=Linear}
    alphaScaleBiasClamp {bInterpResult=true}
    boneToModify {boneName=dyn_hip_bandage_center_left_mstr}
    translation {z=2}
   }
   n13 = "Transform (Modify) Bone - Bone: dyn_hip_bandage_center_right_mstr"(ComponentPose=@n12, Alpha=@n11, Translation="0,0,2", Rotation="0,0,0", Scale="1,1,1") {
    rotationMode=BMM_Replace
    rotationSpace=BCS_BoneSpace
    scaleMode=BMM_Replace
    scaleSpace=BCS_BoneSpace
    translationMode=BMM_Replace
    translationSpace=BCS_BoneSpace
    alphaBoolBlend {blendOption=Linear}
    alphaScaleBiasClamp {bInterpResult=true}
    boneToModify {boneName=dyn_hip_bandage_center_right_mstr}
    translation {z=2}
   }
   n14 = Anim Dynamics - Bone: dyn_hip_bandage_center_left_01!(ComponentPose=@n13, Alpha=1, GravityOverride:Vector="0,0,-980", ExternalForce="0,0,0") -> (Pose) {
    angularDampingOverride=1
    angularSpringConstant=50
    bAngularSpring=true
    bOverrideAngularDamping=true
    bUseGravityOverride=true
    gravityScale=0.3
    simulationSpace=BoneRelative
    alphaBoolBlend {blendOption=Linear}
    boundBone {boneName=dyn_hip_bandage_center_left_01}
    componentAppliedLinearAccClamp=<2500 2500 1500>
    componentLinearAccScale=<0.4 0.4 1>
    componentLinearVelScale=<0.55 0.55 0.6>
    physicsBodyDefinitions [{
      collisionType=CustomSphere
      sphereCollisionRadius=2
      boundBone {boneName=dyn_hip_bandage_center_left_01}
      boxExtents=<20 20 20>
      constraintSetup {angularLimitsMax=<0 0 5>, angularLimitsMin=<0 -30 -5>, angularTarget=<0 0 0>}
      localJointOffset=<9 0 0>
     }]
    planarLimits [{drivingBone {boneName=thigh_l}, planeTransform {location=<0 7.5 0>, rotation=<rot 0 0 90>}}]
    relativeSpaceBone {boneName=root}
   }
   n15 = Anim Dynamics - Bone: dyn_hip_bandage_center_left_02!(ComponentPose=@n14, Alpha=1, GravityOverride:Vector="0,0,-980", ExternalForce="0,0,0") -> (Pose) {
    angularDampingOverride=1
    angularSpringConstant=50
    bAngularSpring=true
    bOverrideAngularDamping=true
    bUseGravityOverride=true
    gravityScale=0.3
    simulationSpace=BoneRelative
    alphaBoolBlend {blendOption=Linear}
    boundBone {boneName=dyn_hip_bandage_center_left_02}
    componentAppliedLinearAccClamp=<2500 2500 1500>
    componentLinearAccScale=<0.4 0.4 1>
    componentLinearVelScale=<0.55 0.55 0.6>
    physicsBodyDefinitions [{
      collisionType=CustomSphere
      sphereCollisionRadius=2
      boundBone {boneName=dyn_hip_bandage_center_left_02}
      boxExtents=<20 20 20>
      constraintSetup {angularLimitsMax=<0 0 5>, angularLimitsMin=<0 -30 -5>, angularTarget=<0 0 0>}
      localJointOffset=<9 0 0>
     }]
    planarLimits [{drivingBone {boneName=thigh_l}, planeTransform {location=<0 7.5 0>, rotation=<rot 0 0 90>}}]
    relativeSpaceBone {boneName=root}
   }
   n16 = Anim Dynamics - Bone: dyn_hip_bandage_center_right_01!(ComponentPose=@n15, Alpha=1, GravityOverride:Vector="0,0,-980", ExternalForce="0,0,0") -> (Pose) {
    angularDampingOverride=1
    angularSpringConstant=50
    bAngularSpring=true
    bOverrideAngularDamping=true
    bUseGravityOverride=true
    gravityScale=0.3
    simulationSpace=BoneRelative
    alphaBoolBlend {blendOption=Linear}
    boundBone {boneName=dyn_hip_bandage_center_right_01}
    componentAppliedLinearAccClamp=<2500 2500 1500>
    componentLinearAccScale=<0.4 0.4 1>
    componentLinearVelScale=<0.55 0.55 0.6>
    physicsBodyDefinitions [{
      collisionType=CustomSphere
      sphereCollisionRadius=2
      boundBone {boneName=dyn_hip_bandage_center_right_01}
      boxExtents=<20 20 20>
      constraintSetup {angularLimitsMax=<0 0 5>, angularLimitsMin=<0 -30 -5>, angularTarget=<0 0 0>}
      localJointOffset=<9 0 0>
     }]
    planarLimits [{drivingBone {boneName=thigh_r}, planeTransform {location=<0 -7.5 0>, rotation=<rot 0 0 -90>}}]
    relativeSpaceBone {boneName=root}
   }
   n17 = Anim Dynamics - Bone: dyn_hip_bandage_center_right_02!(ComponentPose=@n16, Alpha=1, GravityOverride:Vector="0,0,-980", ExternalForce="0,0,0") -> (Pose) {
    angularDampingOverride=1
    angularSpringConstant=50
    bAngularSpring=true
    bOverrideAngularDamping=true
    bUseGravityOverride=true
    gravityScale=0.3
    simulationSpace=BoneRelative
    alphaBoolBlend {blendOption=Linear}
    boundBone {boneName=dyn_hip_bandage_center_right_02}
    componentAppliedLinearAccClamp=<2500 2500 1500>
    componentLinearAccScale=<0.4 0.4 1>
    componentLinearVelScale=<0.55 0.55 0.6>
    physicsBodyDefinitions [{
      collisionType=CustomSphere
      sphereCollisionRadius=2
      boundBone {boneName=dyn_hip_bandage_center_right_02}
      boxExtents=<20 20 20>
      constraintSetup {angularLimitsMax=<0 0 5>, angularLimitsMin=<0 -30 -5>, angularTarget=<0 0 0>}
      localJointOffset=<9 0 0>
     }]
    planarLimits [{drivingBone {boneName=thigh_r}, planeTransform {location=<0 -7.5 0>, rotation=<rot 0 0 -90>}}]
    relativeSpaceBone {boneName=root}
   }
   n18 = Get Jumping Alpha()
   n19 = "Transform (Modify) Bone - Bone: dyn_skirt_fr_01_l_mstr"(ComponentPose=@n17, Alpha=@n18, Translation="0,0,2", Rotation="1,0,0", Scale="1,1,1") {
    rotationMode=BMM_Replace
    rotationSpace=BCS_BoneSpace
    scaleMode=BMM_Replace
    scaleSpace=BCS_BoneSpace
    translationMode=BMM_Replace
    translationSpace=BCS_BoneSpace
    alphaBoolBlend {blendOption=Linear}
    alphaScaleBiasClamp {bInterpResult=true}
    boneToModify {boneName=dyn_skirt_fr_01_l_mstr}
    rotation {pitch=1}
    translation {z=2}
   }
   n20 = "Transform (Modify) Bone - Bone: dyn_skirt_fr2_01_l_mstr"(ComponentPose=@n19, Alpha=@n18, Translation="0,0,2", Rotation="1,0,0", Scale="1,1,1") {
    rotationMode=BMM_Replace
    rotationSpace=BCS_BoneSpace
    scaleMode=BMM_Replace
    scaleSpace=BCS_BoneSpace
    translationMode=BMM_Replace
    translationSpace=BCS_BoneSpace
    alphaBoolBlend {blendOption=Linear}
    alphaScaleBiasClamp {bInterpResult=true}
    boneToModify {boneName=dyn_skirt_fr2_01_l_mstr}
    rotation {pitch=1}
    translation {z=2}
   }
   n21 = "Transform (Modify) Bone - Bone: dyn_skirt_fr_01_r_mstr"(ComponentPose=@n20, Alpha=@n18, Translation="0,0,2", Rotation="1,0,0", Scale="1,1,1") {
    rotationMode=BMM_Replace
    rotationSpace=BCS_BoneSpace
    scaleMode=BMM_Replace
    scaleSpace=BCS_BoneSpace
    translationMode=BMM_Replace
    translationSpace=BCS_BoneSpace
    alphaBoolBlend {blendOption=Linear}
    alphaScaleBiasClamp {bInterpResult=true}
    boneToModify {boneName=dyn_skirt_fr_01_r_mstr}
    rotation {pitch=1}
    translation {z=2}
   }
   n22 = "Transform (Modify) Bone - Bone: dyn_skirt_fr2_01_r_mstr"(ComponentPose=@n21, Alpha=@n18, Translation="0,0,2", Rotation="1,0,0", Scale="1,1,1") {
    rotationMode=BMM_Replace
    rotationSpace=BCS_BoneSpace
    scaleMode=BMM_Replace
    scaleSpace=BCS_BoneSpace
    translationMode=BMM_Replace
    translationSpace=BCS_BoneSpace
    alphaBoolBlend {blendOption=Linear}
    alphaScaleBiasClamp {bInterpResult=true}
    boneToModify {boneName=dyn_skirt_fr2_01_r_mstr}
    rotation {pitch=1}
    translation {z=2}
   }
   n23 = Get Not Crouching() -> (Not Crouching:double)
   n24 = "Transform (Modify) Bone - Bone: dyn_skirt_fr_01_l_mstr"(ComponentPose=@n22, Alpha=@n23, Translation="0,0,1", Rotation="1,0,0", Scale="1,1,1") {
    rotationMode=BMM_Replace
    rotationSpace=BCS_BoneSpace
    scaleMode=BMM_Replace
    scaleSpace=BCS_BoneSpace
    translationMode=BMM_Replace
    translationSpace=BCS_BoneSpace
    alphaBoolBlend {blendOption=Linear}
    alphaScaleBiasClamp {bInterpResult=true}
    boneToModify {boneName=dyn_skirt_fr_01_l_mstr}
    rotation {pitch=1}
    translation {z=1}
   }
   n25 = "Transform (Modify) Bone - Bone: dyn_skirt_fr2_01_l_mstr"(ComponentPose=@n24, Alpha=@n23, Translation="0,0,1", Rotation="1,0,0", Scale="1,1,1") {
    rotationMode=BMM_Replace
    rotationSpace=BCS_BoneSpace
    scaleMode=BMM_Replace
    scaleSpace=BCS_BoneSpace
    translationMode=BMM_Replace
    translationSpace=BCS_BoneSpace
    alphaBoolBlend {blendOption=Linear}
    alphaScaleBiasClamp {bInterpResult=true}
    boneToModify {boneName=dyn_skirt_fr2_01_l_mstr}
    rotation {pitch=1}
    translation {z=1}
   }
   n26 = "Transform (Modify) Bone - Bone: dyn_skirt_fr_01_r_mstr"(ComponentPose=@n25, Alpha=@n23, Translation="0,0,1", Rotation="1,0,0", Scale="1,1,1") {
    rotationMode=BMM_Replace
    rotationSpace=BCS_BoneSpace
    scaleMode=BMM_Replace
    scaleSpace=BCS_BoneSpace
    translationMode=BMM_Replace
    translationSpace=BCS_BoneSpace
    alphaBoolBlend {blendOption=Linear}
    alphaScaleBiasClamp {bInterpResult=true}
    boneToModify {boneName=dyn_skirt_fr_01_r_mstr}
    rotation {pitch=1}
    translation {z=1}
   }
   n27 = "Transform (Modify) Bone - Bone: dyn_skirt_fr2_01_r_mstr"(ComponentPose=@n26, Alpha=@n23, Translation="0,0,1", Rotation="1,0,0", Scale="1,1,1") {
    rotationMode=BMM_Replace
    rotationSpace=BCS_BoneSpace
    scaleMode=BMM_Replace
    scaleSpace=BCS_BoneSpace
    translationMode=BMM_Replace
    translationSpace=BCS_BoneSpace
    alphaBoolBlend {blendOption=Linear}
    alphaScaleBiasClamp {bInterpResult=true}
    boneToModify {boneName=dyn_skirt_fr2_01_r_mstr}
    rotation {pitch=1}
    translation {z=1}
   }
   n28 = Component To Local(ComponentPose=@n27)
   n29 = "Constraint - dyn_hip_bandage_left_mstr (P)"(ComponentPose=Local To Component(LocalPose=Use cached pose 'PreCache'()), Alpha=1, ConstraintWeights_0=1) {
    alphaBoolBlend {blendOption=Linear}
    boneToModify {boneName=dyn_hip_bandage_left_mstr}
    constraintSetup [{targetBone {boneName=thigh_l}}]
    constraintWeights [1]
   }
   n30 = "Constraint - dyn_hip_bandage_center_left_mstr (PP)"(ComponentPose=@n29, Alpha=1, ConstraintWeights_0=0.7, ConstraintWeights_1=0.3) {
    alphaBoolBlend {blendOption=Linear}
    boneToModify {boneName=dyn_hip_bandage_center_left_mstr}
    constraintSetup [{targetBone {boneName=thigh_l}}, {targetBone {boneName=thigh_r}}]
    constraintWeights [1,1]
   }
   n31 = "Constraint - dyn_hip_bandage_right_mstr (P)"(ComponentPose=@n30, Alpha=1, ConstraintWeights_0=1) {
    alphaBoolBlend {blendOption=Linear}
    boneToModify {boneName=dyn_hip_bandage_right_mstr}
    constraintSetup [{targetBone {boneName=thigh_r}}]
    constraintWeights [1]
   }
   n32 = "Constraint - dyn_hip_bandage_center_right_mstr (PP)"(ComponentPose=@n31, Alpha=1, ConstraintWeights_0=0.7, ConstraintWeights_1=0.3) {
    alphaBoolBlend {blendOption=Linear}
    boneToModify {boneName=dyn_hip_bandage_center_right_mstr}
    constraintSetup [{targetBone {boneName=thigh_r}}, {targetBone {boneName=thigh_l}}]
    constraintWeights [1,1]
   }
   n33 = Get Bwd Movement() -> (Bwd Movement:double)
   n34 = Trail controller - Bone: dyn_skirt_bk_03!(ComponentPose=@n32, Alpha=@n33, FakeVelocity:Vector="0,0,0") -> (Pose) {
    bActorSpaceFakeVel=true
    bShowTrailLocation=true
    debugLifeTime=0.1
    maxDeltaTime=0.0333
    relaxationSpeedScale=3
    alphaScaleBiasClamp {bInterpResult=true}
    baseJoint {boneName=root}
    planarLimits [
     {drivingBone {boneName=root}, planeTransform {location=<0 0 10>, rotation=<rot 0 0 0>, scale=<1 1 1>}}
     {drivingBone {boneName=spine_03}, planeTransform {location=<0 -20 0>, rotation=<rot 0 15 -89.9999>, scale=<1 1 1>}}
    ]
    rotationLimits [{limitMax=<180 180 180>}, {limitMax=<180 180 180>}]
    rotationOffsets [<0 0 0>, <0 0 0>]
    trailBone {boneName=dyn_skirt_bk_03}
    trailRelaxationSpeed {editorCurveData {keys [{arriveTangent=-72.0739, interpMode=RCIM_Cubic, leaveTangent=-72.0739, tangentMode=RCTM_User, time=0, value=35}, <at 1 5 RCIM_Cubic>]}}
   }
   n35 = Trail controller - Bone: dyn_skirt_bk_03_l!(ComponentPose=@n34, Alpha=@n33, FakeVelocity:Vector="0,0,0") -> (Pose) {
    bActorSpaceFakeVel=true
    bShowTrailLocation=true
    debugLifeTime=0.1
    maxDeltaTime=0.0333
    relaxationSpeedScale=3
    alphaScaleBiasClamp {bInterpResult=true}
    baseJoint {boneName=root}
    planarLimits [
     {drivingBone {boneName=root}, planeTransform {location=<0 0 10>, rotation=<rot 0 0 0>, scale=<1 1 1>}}
     {drivingBone {boneName=spine_03}, planeTransform {location=<0 -20 0>, rotation=<rot 0 15 -89.9999>, scale=<1 1 1>}}
    ]
    rotationLimits [{limitMax=<180 180 180>}, {limitMax=<180 180 180>}]
    rotationOffsets [<0 0 0>, <0 0 0>]
    trailBone {boneName=dyn_skirt_bk_03_l}
    trailRelaxationSpeed {editorCurveData {keys [{arriveTangent=-72.0739, interpMode=RCIM_Cubic, leaveTangent=-72.0739, tangentMode=RCTM_User, time=0, value=35}, <at 1 5 RCIM_Cubic>]}}
   }
   n36 = Trail controller - Bone: dyn_skirt_bk_03_r!(ComponentPose=@n35, Alpha=@n33, FakeVelocity:Vector="0,0,0") -> (Pose) {
    bActorSpaceFakeVel=true
    bShowTrailLocation=true
    debugLifeTime=0.1
    maxDeltaTime=0.0333
    relaxationSpeedScale=3
    alphaScaleBiasClamp {bInterpResult=true}
    baseJoint {boneName=root}
    planarLimits [
     {drivingBone {boneName=root}, planeTransform {location=<0 0 10>, rotation=<rot 0 0 0>, scale=<1 1 1>}}
     {drivingBone {boneName=spine_03}, planeTransform {location=<0 -20 0>, rotation=<rot 0 15 -89.9999>, scale=<1 1 1>}}
    ]
    rotationLimits [{limitMax=<180 180 180>}, {limitMax=<180 180 180>}]
    rotationOffsets [<0 0 0>, <0 0 0>]
    trailBone {boneName=dyn_skirt_bk_03_r}
    trailRelaxationSpeed {editorCurveData {keys [{arriveTangent=-72.0739, interpMode=RCIM_Cubic, leaveTangent=-72.0739, tangentMode=RCTM_User, time=0, value=35}, <at 1 5 RCIM_Cubic>]}}
   }
   n37 = Get Left Movement() -> (Left Movement:double)
   n38 = Trail controller - Bone: dyn_skirt_02_l!(ComponentPose=@n36, Alpha=@n37, FakeVelocity:Vector="0,0,0") -> (Pose) {
    bActorSpaceFakeVel=true
    bShowTrailLocation=true
    debugLifeTime=0.1
    maxDeltaTime=0.0333
    relaxationSpeedScale=3
    alphaScaleBiasClamp {bInterpResult=true}
    baseJoint {boneName=root}
    planarLimits [
     {drivingBone {boneName=root}, planeTransform {location=<0 0 10>, rotation=<rot 0 0 0>, scale=<1 1 1>}}
     {drivingBone {boneName=spine_03}, planeTransform {location=<0 -20 0>, rotation=<rot 0 15 -89.9999>, scale=<1 1 1>}}
    ]
    rotationLimits [{limitMax=<180 180 180>}, {limitMax=<180 180 180>}]
    rotationOffsets [<0 0 0>, <0 0 0>]
    trailBone {boneName=dyn_skirt_02_l}
    trailRelaxationSpeed {editorCurveData {keys [{arriveTangent=-72.0739, interpMode=RCIM_Cubic, leaveTangent=-72.0739, tangentMode=RCTM_User, time=0, value=35}, <at 1 5 RCIM_Cubic>]}}
   }
   n39 = Trail controller - Bone: dyn_skirt_fr2_03_l!(ComponentPose=@n38, Alpha=@n37, FakeVelocity:Vector="0,0,0") -> (Pose) {
    bActorSpaceFakeVel=true
    bShowTrailLocation=true
    debugLifeTime=0.1
    maxDeltaTime=0.0333
    relaxationSpeedScale=3
    alphaScaleBiasClamp {bInterpResult=true}
    baseJoint {boneName=root}
    planarLimits [
     {drivingBone {boneName=root}, planeTransform {location=<0 0 10>, rotation=<rot 0 0 0>, scale=<1 1 1>}}
     {drivingBone {boneName=spine_03}, planeTransform {location=<0 -20 0>, rotation=<rot 0 15 -89.9999>, scale=<1 1 1>}}
    ]
    rotationLimits [{limitMax=<180 180 180>}, {limitMax=<180 180 180>}]
    rotationOffsets [<0 0 0>, <0 0 0>]
    trailBone {boneName=dyn_skirt_fr2_03_l}
    trailRelaxationSpeed {editorCurveData {keys [{arriveTangent=-72.0739, interpMode=RCIM_Cubic, leaveTangent=-72.0739, tangentMode=RCTM_User, time=0, value=35}, <at 1 5 RCIM_Cubic>]}}
   }
   n40 = Get Right Movement() -> (Right Movement:double)
   n41 = Trail controller - Bone: dyn_skirt_02_r!(ComponentPose=@n39, Alpha=@n40, FakeVelocity:Vector="0,0,0") -> (Pose) {
    bActorSpaceFakeVel=true
    bShowTrailLocation=true
    debugLifeTime=0.1
    maxDeltaTime=0.0333
    relaxationSpeedScale=3
    alphaScaleBiasClamp {bInterpResult=true}
    baseJoint {boneName=root}
    planarLimits [
     {drivingBone {boneName=root}, planeTransform {location=<0 0 10>, rotation=<rot 0 0 0>, scale=<1 1 1>}}
     {drivingBone {boneName=spine_03}, planeTransform {location=<0 -20 0>, rotation=<rot 0 15 -89.9999>, scale=<1 1 1>}}
    ]
    rotationLimits [{limitMax=<180 180 180>}, {limitMax=<180 180 180>}]
    rotationOffsets [<0 0 0>, <0 0 0>]
    trailBone {boneName=dyn_skirt_02_r}
    trailRelaxationSpeed {editorCurveData {keys [{arriveTangent=-72.0739, interpMode=RCIM_Cubic, leaveTangent=-72.0739, tangentMode=RCTM_User, time=0, value=35}, <at 1 5 RCIM_Cubic>]}}
   }
   n42 = Trail controller - Bone: dyn_skirt_fr2_03_r!(ComponentPose=@n41, Alpha=@n40, FakeVelocity:Vector="0,0,0") -> (Pose) {
    bActorSpaceFakeVel=true
    bShowTrailLocation=true
    debugLifeTime=0.1
    maxDeltaTime=0.0333
    relaxationSpeedScale=3
    alphaScaleBiasClamp {bInterpResult=true}
    baseJoint {boneName=root}
    planarLimits [
     {drivingBone {boneName=root}, planeTransform {location=<0 0 10>, rotation=<rot 0 0 0>, scale=<1 1 1>}}
     {drivingBone {boneName=spine_03}, planeTransform {location=<0 -20 0>, rotation=<rot 0 15 -89.9999>, scale=<1 1 1>}}
    ]
    rotationLimits [{limitMax=<180 180 180>}, {limitMax=<180 180 180>}]
    rotationOffsets [<0 0 0>, <0 0 0>]
    trailBone {boneName=dyn_skirt_fr2_03_r}
    trailRelaxationSpeed {editorCurveData {keys [{arriveTangent=-72.0739, interpMode=RCIM_Cubic, leaveTangent=-72.0739, tangentMode=RCTM_User, time=0, value=35}, <at 1 5 RCIM_Cubic>]}}
   }
   n43 = Component To Local(ComponentPose=@n42)
   n44 = Get bRigidBodyEnabled()
   n45 = Get Is Custom Emote Playing Break Dance() -> (Is Custom Emote Playing Break Dance:double)
   n46 = "Transform (Modify) Bone - Bone: dyn_skirt_fr_01_r_mstr"(
    ComponentPose=Local To Component(LocalPose=Blend Poses by bool(BlendPose_0=@n28, BlendPose_1=@n43, bActiveValue=@n44) {blendType=Linear})
    Alpha=@n45
    Translation="0,0,6"
    Rotation="5,0,0"
    Scale="1,1,1"
   ) {
    rotationMode=BMM_Replace
    rotationSpace=BCS_BoneSpace
    scaleMode=BMM_Replace
    scaleSpace=BCS_BoneSpace
    translationMode=BMM_Replace
    translationSpace=BCS_BoneSpace
    alphaBoolBlend {blendOption=Linear}
    alphaScaleBiasClamp {bInterpResult=true}
    boneToModify {boneName=dyn_skirt_fr_01_r_mstr}
    rotation {pitch=5}
    translation {z=6}
   }
   n47 = "Transform (Modify) Bone - Bone: dyn_skirt_fr2_01_r_mstr"(ComponentPose=@n46, Alpha=@n45, Translation="0,0,5", Rotation="5,0,0", Scale="1,1,1") {
    rotationMode=BMM_Replace
    rotationSpace=BCS_BoneSpace
    scaleMode=BMM_Replace
    scaleSpace=BCS_BoneSpace
    translationMode=BMM_Replace
    translationSpace=BCS_BoneSpace
    alphaBoolBlend {blendOption=Linear}
    alphaScaleBiasClamp {bInterpResult=true}
    boneToModify {boneName=dyn_skirt_fr2_01_r_mstr}
    rotation {pitch=5}
    translation {z=5}
   }
   n48 = Get Is Custom Emote Playing Dance Moves() -> (Is Custom Emote Playing Dance Moves:double)
   n49 = "Transform (Modify) Bone - Bone: dyn_skirt_fr_01_l_mstr"(ComponentPose=@n47, Alpha=@n48, Translation="0,0,2", Rotation="0,0,0", Scale="1,1,1") {
    rotationMode=BMM_Replace
    rotationSpace=BCS_BoneSpace
    scaleMode=BMM_Replace
    scaleSpace=BCS_BoneSpace
    translationMode=BMM_Replace
    translationSpace=BCS_BoneSpace
    alphaBoolBlend {blendOption=Linear}
    alphaScaleBiasClamp {bInterpResult=true}
    boneToModify {boneName=dyn_skirt_fr_01_l_mstr}
    translation {z=2}
   }
   n50 = "Transform (Modify) Bone - Bone: dyn_skirt_fr_01_l_mstr"(ComponentPose=@n49, Alpha=@n48, Translation="0,0,3", Rotation="0,0,0", Scale="1,1,1") {
    rotationMode=BMM_Replace
    rotationSpace=BCS_BoneSpace
    scaleMode=BMM_Replace
    scaleSpace=BCS_BoneSpace
    translationMode=BMM_Replace
    translationSpace=BCS_BoneSpace
    alphaBoolBlend {blendOption=Linear}
    alphaScaleBiasClamp {bInterpResult=true}
    boneToModify {boneName=dyn_skirt_fr_01_l_mstr}
    translation {z=3}
   }
   n51 = "Transform (Modify) Bone - Bone: dyn_skirt_fr_01_r_mstr"(ComponentPose=@n50, Alpha=@n48, Translation="0,0,2", Rotation="2.5,0,0", Scale="1,1,1") {
    rotationMode=BMM_Replace
    rotationSpace=BCS_BoneSpace
    scaleMode=BMM_Replace
    scaleSpace=BCS_BoneSpace
    translationMode=BMM_Replace
    translationSpace=BCS_BoneSpace
    alphaBoolBlend {blendOption=Linear}
    alphaScaleBiasClamp {bInterpResult=true}
    boneToModify {boneName=dyn_skirt_fr_01_r_mstr}
    rotation {pitch=2.5}
    translation {z=2}
   }
   n52 = "Transform (Modify) Bone - Bone: dyn_skirt_fr2_01_r_mstr"(ComponentPose=@n51, Alpha=@n48, Translation="0,0,2", Rotation="2.5,0,0", Scale="1,1,1") {
    rotationMode=BMM_Replace
    rotationSpace=BCS_BoneSpace
    scaleMode=BMM_Replace
    scaleSpace=BCS_BoneSpace
    translationMode=BMM_Replace
    translationSpace=BCS_BoneSpace
    alphaBoolBlend {blendOption=Linear}
    alphaScaleBiasClamp {bInterpResult=true}
    boneToModify {boneName=dyn_skirt_fr2_01_r_mstr}
    rotation {pitch=2.5}
    translation {z=2}
   }
   n53 = "Transform (Modify) Bone - Bone: dyn_hip_bandage_center_right_mstr"(ComponentPose=@n52, Alpha=@n48, Translation="0,0,2", Rotation="2.5,0,0", Scale="1,1,1") {
    rotationMode=BMM_Replace
    rotationSpace=BCS_BoneSpace
    scaleMode=BMM_Replace
    scaleSpace=BCS_BoneSpace
    translationMode=BMM_Replace
    translationSpace=BCS_BoneSpace
    alphaBoolBlend {blendOption=Linear}
    alphaScaleBiasClamp {bInterpResult=true}
    boneToModify {boneName=dyn_hip_bandage_center_right_mstr}
    rotation {pitch=2.5}
    translation {z=2}
   }
   n54 = "Transform (Modify) Bone - Bone: dyn_hip_bandage_center_left_mstr"(ComponentPose=@n53, Alpha=@n48, Translation="0,0,2", Rotation="2.5,0,0", Scale="1,1,1") {
    rotationMode=BMM_Replace
    rotationSpace=BCS_BoneSpace
    scaleMode=BMM_Replace
    scaleSpace=BCS_BoneSpace
    translationMode=BMM_Replace
    translationSpace=BCS_BoneSpace
    alphaBoolBlend {blendOption=Linear}
    alphaScaleBiasClamp {bInterpResult=true}
    boneToModify {boneName=dyn_hip_bandage_center_left_mstr}
    rotation {pitch=2.5}
    translation {z=2}
   }
   Output Pose(Result=Component To Local(ComponentPose=@n54)) {layerGroup=None}
   n55 = "Constraint - dyn_skirt_fr_01_l_mstr (PPPPPPPP)"!(
    ComponentPose=Local To Component(LocalPose=Copy Pose From Mesh() {bUseAttachedParent=true})
    Alpha=1
    ConstraintWeights_0=0.3425
    ConstraintWeights_1:float=0.0784
    ConstraintWeights_2:float=0.229
    ConstraintWeights_3:float=0.0841
    ConstraintWeights_4:float=0.0259
    ConstraintWeights_5:float=0.1535
    ConstraintWeights_6:float=0.0573
    ConstraintWeights_7:float=0.0293
   ) -> (Pose) {
    alphaBoolBlend {blendOption=Linear}
    boneToModify {boneName=dyn_skirt_fr_01_l_mstr}
    constraintSetup [
     {targetBone {boneName=pelvis}}
     {targetBone {boneName=thigh_l}}
     {targetBone {boneName=thigh_twist_01_l}}
     {targetBone {boneName=deform_groin_base_l}}
     {targetBone {boneName=thigh_r}}
     {targetBone {boneName=thigh_twist_01_r}}
     {targetBone {boneName=deform_groin_r}}
     {targetBone {boneName=deform_glute_l}}
    ]
    constraintWeights [1,1,1,1,1,1,1,1]
   }
   n56 = "Constraint - dyn_skirt_fr2_01_l_mstr (PPPPPPPP)"!(
    ComponentPose=@n55
    Alpha=1
    ConstraintWeights_0=0.271
    ConstraintWeights_1:float=0.1326
    ConstraintWeights_2:float=0.3626
    ConstraintWeights_3:float=0.0805
    ConstraintWeights_4:float=0.0737
    ConstraintWeights_5:float=0.0248
    ConstraintWeights_6:float=0.0289
    ConstraintWeights_7:float=0.0259
   ) -> (Pose) {
    alphaBoolBlend {blendOption=Linear}
    boneToModify {boneName=dyn_skirt_fr2_01_l_mstr}
    constraintSetup [
     {targetBone {boneName=pelvis}}
     {targetBone {boneName=thigh_l}}
     {targetBone {boneName=thigh_twist_01_l}}
     {targetBone {boneName=deform_groin_base_l}}
     {targetBone {boneName=thigh_twist_01_r}}
     {targetBone {boneName=deform_groin_r}}
     {targetBone {boneName=deform_glute_l}}
     {targetBone {boneName=deform_glute_base_r}}
    ]
    constraintWeights [1,1,1,1,1,1,1,1]
   }
   n57 = "Constraint - dyn_skirt_01_l_mstr (PPPPPPPP)"!(
    ComponentPose=@n56
    Alpha=1
    ConstraintWeights_0=0.2669
    ConstraintWeights_1:float=0.1111
    ConstraintWeights_2:float=0.4125
    ConstraintWeights_3:float=0.0354
    ConstraintWeights_4:float=0.0707
    ConstraintWeights_5:float=0.0237
    ConstraintWeights_6:float=0.0528
    ConstraintWeights_7:float=0.027
   ) -> (Pose) {
    alphaBoolBlend {blendOption=Linear}
    boneToModify {boneName=dyn_skirt_01_l_mstr}
    constraintSetup [
     {targetBone {boneName=pelvis}}
     {targetBone {boneName=thigh_l}}
     {targetBone {boneName=thigh_twist_01_l}}
     {targetBone {boneName=deform_groin_base_l}}
     {targetBone {boneName=thigh_twist_01_r}}
     {targetBone {boneName=deform_groin_r}}
     {targetBone {boneName=deform_glute_l}}
     {targetBone {boneName=deform_glute_base_r}}
    ]
    constraintWeights [1,1,1,1,1,1,1,1]
   }
   n58 = "Constraint - dyn_skirt_fr_01_r_mstr (PPPPPPPP)"!(
    ComponentPose=@n57
    Alpha=1
    ConstraintWeights_0=0.3421
    ConstraintWeights_1:float=0.0276
    ConstraintWeights_2:float=0.1461
    ConstraintWeights_3:float=0.0522
    ConstraintWeights_4:float=0.0665
    ConstraintWeights_5:float=0.2423
    ConstraintWeights_6:float=0.0942
    ConstraintWeights_7:float=0.029
   ) -> (Pose) {
    alphaBoolBlend {blendOption=Linear}
    boneToModify {boneName=dyn_skirt_fr_01_r_mstr}
    constraintSetup [
     {targetBone {boneName=pelvis}}
     {targetBone {boneName=thigh_l}}
     {targetBone {boneName=thigh_twist_01_l}}
     {targetBone {boneName=deform_groin_base_l}}
     {targetBone {boneName=thigh_r}}
     {targetBone {boneName=thigh_twist_01_r}}
     {targetBone {boneName=deform_groin_r}}
     {targetBone {boneName=deform_glute_l}}
    ]
    constraintWeights [1,1,1,1,1,1,1,1]
   }
   n59 = "Constraint - dyn_skirt_fr2_01_r_mstr (PPPPPPPP)"!(
    ComponentPose=@n58
    Alpha=1
    ConstraintWeights_0=0.3063
    ConstraintWeights_1:float=0.0936
    ConstraintWeights_2:float=0.0311
    ConstraintWeights_3:float=0.1022
    ConstraintWeights_4:float=0.312
    ConstraintWeights_5:float=0.1092
    ConstraintWeights_6:float=0.0277
    ConstraintWeights_7:float=0.0178
   ) -> (Pose) {
    alphaBoolBlend {blendOption=Linear}
    boneToModify {boneName=dyn_skirt_fr2_01_r_mstr}
    constraintSetup [
     {targetBone {boneName=pelvis}}
     {targetBone {boneName=thigh_twist_01_l}}
     {targetBone {boneName=deform_groin_base_l}}
     {targetBone {boneName=thigh_r}}
     {targetBone {boneName=thigh_twist_01_r}}
     {targetBone {boneName=deform_groin_r}}
     {targetBone {boneName=deform_glute_l}}
     {targetBone {boneName=deform_glute_base_r}}
    ]
    constraintWeights [1,1,1,1,1,1,1,1]
   }
   n60 = "Constraint - dyn_skirt_01_r_mstr (PPPPPPPP)"!(
    ComponentPose=@n59
    Alpha=1
    ConstraintWeights_0=0.269
    ConstraintWeights_1:float=0.0712
    ConstraintWeights_2:float=0.0227
    ConstraintWeights_3:float=0.1055
    ConstraintWeights_4:float=0.4135
    ConstraintWeights_5:float=0.037
    ConstraintWeights_6:float=0.0274
    ConstraintWeights_7:float=0.0538
   ) -> (Pose) {
    alphaBoolBlend {blendOption=Linear}
    boneToModify {boneName=dyn_skirt_01_r_mstr}
    constraintSetup [
     {targetBone {boneName=pelvis}}
     {targetBone {boneName=thigh_twist_01_l}}
     {targetBone {boneName=deform_groin_base_l}}
     {targetBone {boneName=thigh_r}}
     {targetBone {boneName=thigh_twist_01_r}}
     {targetBone {boneName=deform_groin_r}}
     {targetBone {boneName=deform_glute_l}}
     {targetBone {boneName=deform_glute_base_r}}
    ]
    constraintWeights [1,1,1,1,1,1,1,1]
   }
   n61 = Control Rig!(
    Source=Component To Local(ComponentPose=@n60)
    Alpha
    Get Crouching:bool=Get bIsCrouching()
    Get Accelerating:bool=Get bIsAccelerating2D()
    Get RBAN:bool=Get bRigidBodyEnabled()
    Get Falling:bool=Get bIsFalling()
    Get Jumping:bool=Get bIsJumping()
    Get Sprinting:bool=Get bIsSprinting()
    Get Emoting:bool=Get bIsPlayingEmote()
   ) -> (Pose) {
    alphaBoolBlend {blendOption=Linear}
    controlRigAssetReference {blueprintRigClass=&/F_MED_PlantStand_Iris_Control_Rig.F_MED_PlantStand_Iris_Control_Rig_C}
   }
   Save cached pose 'PreCache'(Pose=@n61) {cacheName=PreCache}
  ]}
 EventGraph {nodes [
   Event BlueprintUpdateAnimation()
   > n1 = Sequence!(execute) -> (
    then_0
    then_1
    then_2:Exec
    then_3:Exec
    then_4:Exec
    then_5:Exec
    then_6:Exec
    then_7:Exec
    then_8:Exec
    then_9:Exec
    then_10:Exec
    then_11:Exec
   )
   n2 = GetAnimInstance(self=GetOwnerSkeletalMeshComponent())
   > Is Valid(InputObject=@n2)
   > Set Is Custom Emote Playing Break Dance(Is Custom Emote Playing Break Dance:double=SelectFloat(A=1, B=0, bPickA=Montage_IsPlaying(self=@n2, Montage=/Game/Animation/Game/MainPlayer/Emotes/Emote_Breakdance_Female_M))) -> (Output_Get:double)
   n3 = GetAnimInstance(self=GetOwnerSkeletalMeshComponent())
   Is Valid(exec=@n1.then_1, InputObject=@n3)
   > Set Is Custom Emote Playing Dance Moves(Is Custom Emote Playing Dance Moves:double=SelectFloat(A=1, B=0, bPickA=Montage_IsPlaying(self=@n3, Montage=/Game/Animation/Game/MainPlayer/Montages/Emotes/Emote_DanceMoves))) -> (Output_Get:double)
   n4 = PlayerMovement(execute=@n1.then_10, AnimInstance=Self-Reference(), XDotProductCoefficient=1, YDotProductCoefficient=1, ZDotProductCoefficient=1)
   n5 = MapRangeClamped(Value=@n4.YMovement, InRangeA=0, InRangeB=0)
   n6 = float * float(A="Absolute (Float)"(A=@n4), B=0.5)
   > Set Bwd Movement(Bwd Movement:double=MapRangeClamped(
     Value=float + float(A=@n6, B=@n5) {is=title K2Node_CommutativeAssociativeBinaryOperator}
     InRangeA=0
     InRangeB=1.1
     OutRangeA=0.1
     OutRangeB=0
    )) -> (Output_Get:double)
   n7 = MapRangeClamped(Value=@n4, InRangeA=-0.2, InRangeB=1)
   n8 = float * float(A="Absolute (Float)"(A=@n4.YMovement), B=0.2)
   > Set Right Movement(Right Movement:double=MapRangeClamped(Value=float + float(A=@n8, B=@n7), InRangeA=0, InRangeB=1.2, OutRangeB=0)) -> (Output_Get:double)
   n9 = MapRangeClamped(Value=@n4, InRangeA=0, InRangeB=0, OutRangeB=0.3)
   n10 = float * float(A="Absolute (Float)"(A=@n4.YMovement), B=0.2)
   > Set Left Movement(Left Movement:double=MapRangeClamped(Value=float + float(A=@n10, B=@n9), InRangeA=0.2, InRangeB=1.2, OutRangeB=0.5)) -> (Output_Get:double)
   n11 = Gravity Override State Function DEPRECATED(
    execute=@n1.then_2
    CharacterPartAnimInstance=Self-Reference()
    GameTimeInSeconds=GetGameTimeInSeconds()
    GravityOverrideParams=Get Gravity Override Params()
   )
   > Set Gravity Override(Gravity Override=@n11)
   > Set Front End Alpha(Front End Alpha:double=SelectFloat(A=0.5, B=1, bPickA=Get bIsInFrontEnd())) -> (Output_Get:double)
   Set Crouching Alpha(execute=@n1.then_3, Crouching Alpha:double=SelectFloat(A=1, B=0, bPickA=Get bIsCrouching())) -> (Output_Get:double)
   n12 = Get bIsCrouching()
   Set Not Crouching(execute=@n1.then_4, Not Crouching:double=SelectFloat(A=1, B=0.5, bPickA=NOT Boolean(A=@n12))) -> (Output_Get:double)
   n13 = Get bIsJumping()
   n14 = Get bIsFalling()
   Set Jumping Alpha(execute=@n1.then_5, Jumping Alpha:double=SelectFloat(A=1, B=0, bPickA=OR Boolean(A=@n13, B=@n14))) -> (Output_Get:double)
   Is Valid(exec=@n1.then_6, InputObject=Get Player Pawn() -> (Player Pawn:Fort Player Pawn Athena Object Reference))
   > n15 = Branch(Condition=Get bIsInFrontEnd())
   Set WPO Globals Function(
    execute=@n15.else
    Material=Get Mat Instance Dynamic()
    Overlay Material=Get Overlay Mat Instance Dynamic()
    Skeletal Mesh=Get Skel Mesh()
    "Local Noise Transform  Location"=Get Local Noise Transform Location() -> (Local Noise Transform Location:Vector)
    Local Noise Offset=Get Local Noise Offset()
   )
   n16 = Reroute Node(InputPin:double) -> (OutputPin:double)
   > Set WPO Regional Function!(
    execute
    Material:Material Instance Dynamic Object Reference=Get Mat Instance Dynamic()
    Overlay Material:Material Instance Dynamic Object Reference=Get Overlay Mat Instance Dynamic()
    Skeletal Mesh:Skeletal Mesh Component Object Reference=Get Skel Mesh()
    Velocity Transform:Name=foot_r
    Velocity Interp Speed:double=Get Velocity Interp Speed()
    Collision Transform:Name=calf_r
    Collision Transform Axis:EAxis Enum
    Invert Collision Transform Axis:bool=false
    Delta Time X:double=@n16
    Previous Velocity Transform Location:"Vector (by ref)"=Get R1 Velocity Transform Location()
    Previous Velocity Transform Velocity:"Vector (by ref)"=Get R1 Velocity Transform Velocity()
    Region:int=1
   ) -> (then)
   > Set WPO Regional Function(
    Material=Get Mat Instance Dynamic()
    Overlay Material=Get Overlay Mat Instance Dynamic()
    Skeletal Mesh=Get Skel Mesh()
    Velocity Transform=foot_l
    Velocity Interp Speed=Get Velocity Interp Speed()
    Collision Transform=calf_l
    Invert Collision Transform Axis=true
    Delta Time X=@n16
    Previous Velocity Transform Location=Get R2 Velocity Transform Location()
    Previous Velocity Transform Velocity=Get R2 Velocity Transform Velocity()
    Region=2
   )
   > n17 = Branch(Condition=Get bIsSkydiveFloating())
   > n18 = Set WPO Pho Wind Intensity(WPO Pho Wind Intensity=1)
   SetScalarParameterValue(
    execute=[@n19.then, @n18.then]
    self=[Get Mat Instance Dynamic(), Get Overlay Mat Instance Dynamic()]
    ParameterName=ClothSim_WindStrength
    Value=Get WPO Pho Wind Intensity()
   )
   n19 = Set WPO Pho Wind Intensity(execute=@n17.else, WPO Pho Wind Intensity=0)
   n20 = Get bIsJumping()
   n21 = Get bIsFalling()
   Set Not Jumping Alpha(execute=@n1.then_7, Not Jumping Alpha:double=SelectFloat(A=0, B=1, bPickA=OR Boolean(A=@n20, B=@n21))) -> (Output_Get:double)
   n22 = Branch(execute=@n1.then_8, Condition=Get bIsSkyDiving())
   > Set skydiving new(skydiving new:double=1) -> (Output_Get:double)
   Set skydiving new(execute=@n22.else, skydiving new=0)
   n23 = Branch(execute=@n1.then_9, Condition=Get bIsSkyDiving())
   > n24 = Branch(Condition="float > float"(A=Get PawnVelocityZ(), B=50) {is=title K2Node_CallFunction})
   > Set Bwd Movement(Bwd Movement=float * float(A=Get Bwd Movement(), B=0))
   Set Bwd Movement(execute=[@n23.else, @n24.else], Bwd Movement=float * float(A=Get Bwd Movement(), B=1))
   Event BlueprintInitializeAnimation()
   > Set Skel Mesh(Skel Mesh=GetOwnerSkeletalMeshComponent())
   n25 = TryGetPawnOwner()
   > Is Valid(InputObject=@n25)
   > n26 = Cast To FortPlayerPawnAthena(Object=@n25)
   > Set Player Pawn(Player Pawn:Fort Player Pawn Athena Object Reference=@n26) -> (Output_Get:Fort Player Pawn Athena Object Reference)
   Event OnCharacterCustomizationLoaded()
   > Sequence!(execute) -> (then_0, then_1)
   > n27 = Set Skel Mesh(Skel Mesh=GetPartSkeletalMeshComponent())
   > n28 = Get Dynamic Material Instance(SkeletalMeshComponent=@n27)
   > Set Mat Instance Dynamic(Mat Instance Dynamic=@n28)
   > n29 = GetOrCreateDynamicOverlayMaterial(MeshComponent=@n27, MaterialName=None)
   > Set Overlay Mat Instance Dynamic(Overlay Mat Instance Dynamic=@n29)
   MapRangeClamped(Value=0, InRangeA=0, InRangeB=1.2, OutRangeB=0.5)
  ]}
]
variables [
 Bwd Movement {type=number}
 Crouch Walking {type=number}
 Crouching Alpha {type=number}
 Front End Alpha {type=number}
 Gravity Override {type=Vector}
 Gravity Override Params {type=GravityOverrideParamsStruct}
 Is Custom Emote Playing Break Dance {type=number}
 Is Custom Emote Playing Dance Moves {type=number}
 Jumping Alpha {type=number}
 Left Movement {type=number}
 Local Noise Offset {type=Vector}
 Local Noise Transform Location {type=Vector}
 Mat Instance Dynamic {type=MaterialInstanceDynamic}
 Not Crouching {type=number}
 Not Jumping Alpha {type=number}
 Overlay Mat Instance Dynamic {type=MaterialInstanceDynamic}
 Player Pawn {type=FortPlayerPawnAthena}
 R1 Velocity Transform Location {type=Vector}
 R1 Velocity Transform Velocity {type=Vector}
 R2 Velocity Transform Location {type=Vector}
 R2 Velocity Transform Velocity {type=Vector}
 Right Movement {type=number}
 Skel Mesh {type=SkeletalMeshComponent}
 Skydiving New {type=number, default=0.3}
 Velocity Interp Speed {type=number}
 WPO Pho Wind Intensity {type=number}
 WPO Pho Wind Intensity_0 {type=number}
]
