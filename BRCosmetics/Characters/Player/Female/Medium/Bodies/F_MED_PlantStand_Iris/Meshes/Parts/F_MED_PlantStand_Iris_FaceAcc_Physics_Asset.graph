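# Physics asset for the FaceAcc part of the female PlantStand Iris cosmetic body, defining secondary-motion physics for its hair.
# Two head capsules form the skull volume, while small spheres on master bones and capsules on strand bones cover bangs, front, side and back locks.
# Master links from the head down have free swing and twist; strand joints branch from them with tight angular limits.
# Three extra free-angle constraints tie neighboring side and back strands together.
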PhysicsAsset /BRCosmetics/Characters/Player/Female/Medium/Bodies/F_MED_PlantStand_Iris/Meshes/Parts/F_MED_PlantStand_Iris_FaceAcc_Physics_Asset
bodies {
 dyn_front__left_hair_01_b [{shapeType=Capsule, length=3, radius=2, shapeName=dyn_front__left_hair_01_b_capsule, center=<3 0 0.5>, rotation=<rot 90 0 0>}]
 dyn_front_left_hair_01 [{shapeType=Capsule, length=3, radius=2, shapeName=dyn_front_left_hair_01_capsule, center=<3 0 0.5>, rotation=<rot 90 0 0>}]
 dyn_front_left_hair_01_a [{shapeType=Capsule, length=3, radius=2, shapeName=dyn_front__left_hair_01_b_capsule, center=<3 0 0.5>, rotation=<rot 90 0 0>}]
 dyn_front_left_hair_01_c [{shapeType=Capsule, length=3, radius=2, shapeName=dyn_front__left_hair_01_b_capsule, center=<3 0 0.5>, rotation=<rot 90 0 0>}]
 dyn_front_left_hair_01_mstr [{shapeType=Sphere, radius=0.5, shapeName=dyn_front_left_hair_01_mstr_sphere}]
 dyn_hair_01 [{shapeType=Sphere, radius=0.5, shapeName=dyn_hair_01_sphere}]
 dyn_hair_bang_01 [{shapeType=Capsule, length=3, radius=1, shapeName=dyn_front__left_hair_01_b_capsule, center=<3 0 0.5>, rotation=<rot 90 0 0>}]
 dyn_hair_bang_01_mstr [{shapeType=Sphere, radius=0.5, shapeName=dyn_hair_bang_01_mstr_sphere}]
 dyn_hair_front_right_01 [{shapeType=Capsule, length=4, radius=2, shapeName=dyn_front__left_hair_01_b_capsule, center=<4 0 0.5>, rotation=<rot 90 0 0>}]
 dyn_hair_front_right_01_m [{shapeType=Capsule, length=3, radius=2, shapeName=dyn_front__left_hair_01_b_capsule, center=<3 0 1>, rotation=<rot 90 0 0>}]
 dyn_hair_front_right_01_mstr [{shapeType=Sphere, radius=0.5, shapeName=dyn_hair_front_right_01_mstr_sphere}]
 dyn_hair_front_right_01_n [{shapeType=Capsule, length=3, radius=2, shapeName=dyn_front__left_hair_01_b_capsule, center=<3 0 1>, rotation=<rot 90 0 0>}]
 dyn_hair_mstr [{shapeType=Sphere, radius=0.5, shapeName=dyn_hair_mstr_sphere}]
 dyn_left_back_hair_01 [{shapeType=Capsule, length=3, radius=2, shapeName=dyn_front__left_hair_01_b_capsule, center=<3 0 0.5>, rotation=<rot 90 0 0>}]
 dyn_left_back_hair_01_e [{shapeType=Capsule, length=2, radius=2, shapeName=dyn_front__left_hair_01_b_capsule, center=<3 0 1>, rotation=<rot 90 0 0>}]
 dyn_left_back_hair_01_f [{shapeType=Capsule, length=3, radius=2, shapeName=dyn_front__left_hair_01_b_capsule, center=<4 0 2.5>, rotation=<rot 90 0 0>}]
 dyn_left_back_hair_01_g [{shapeType=Capsule, length=2, radius=2, shapeName=dyn_front__left_hair_01_b_capsule, center=<3 0 1>, rotation=<rot 90 0 0>}]
 dyn_left_back_hair_01_mstr [{shapeType=Sphere, radius=0.5, shapeName=dyn_left_back_hair_01_mstr_sphere}]
 dyn_left_hair_01_mstr [{shapeType=Sphere, radius=0.5, shapeName=dyn_left_hair_01_mstr_sphere}]
 dyn_left_hair_02 [{shapeType=Capsule, length=3, radius=2, shapeName=dyn_front__left_hair_01_b_capsule, center=<3 0 1>, rotation=<rot 90 0 0>}]
 dyn_left_hair_02_d [{shapeType=Capsule, length=2, radius=2, shapeName=dyn_front__left_hair_01_b_capsule, center=<2.5 0 1>, rotation=<rot 90 0 0>}]
 dyn_right_back_hair_01 [{shapeType=Capsule, length=3, radius=2, shapeName=dyn_front__left_hair_01_b_capsule, center=<3 0 1>, rotation=<rot 90 0 0>}]
 dyn_right_back_hair_01_h [{shapeType=Capsule, length=3, radius=2, shapeName=dyn_front__left_hair_01_b_capsule, center=<3 0 1>, rotation=<rot 90 0 0>}]
 dyn_right_back_hair_01_i [{shapeType=Capsule, length=3, radius=2, shapeName=dyn_front__left_hair_01_b_capsule, center=<3 0 1>, rotation=<rot 90 0 0>}]
 dyn_right_back_hair_01_mstr [{shapeType=Sphere, radius=0.5, shapeName=dyn_right_back_hair_01_mstr_sphere}]
 dyn_right_hair_01 [{shapeType=Capsule, length=3, radius=2, shapeName=dyn_front__left_hair_01_b_capsule, center=<4 0 1>, rotation=<rot 90 0 0>}]
 dyn_right_hair_01_j [{shapeType=Capsule, length=2, radius=2, shapeName=dyn_front__left_hair_01_b_capsule, center=<2.5 0 1>, rotation=<rot 90 0 0>}]
 dyn_right_hair_01_k [{shapeType=Capsule, length=2, radius=2, shapeName=dyn_front__left_hair_01_b_capsule, center=<2.5 0 1>, rotation=<rot 90 0 0>}]
 dyn_right_hair_01_l [{shapeType=Capsule, length=2, radius=2, shapeName=dyn_front__left_hair_01_b_capsule, center=<2.5 0 1>, rotation=<rot 90 0 0>}]
 dyn_right_hair_01_mstr [{shapeType=Sphere, radius=0.5, shapeName=dyn_right_hair_01_mstr_sphere}]
 head [
  {shapeType=Capsule, length=8, radius=7, center=<0 -0.5 0>, rotation=<rot 90 0 0>}
  {shapeType=Capsule, length=8, radius=6, center=<0 3 0>, rotation=<rot 90 0 0>}
 ]
}
constraints [
 {
  bone1Name=dyn_hair_01
  bone2Name=dyn_hair_mstr
  swing1LimitDegrees=45
  swing1Motion=Free
  swing2LimitDegrees=45
  swing2Motion=Free
  twistLimitDegrees=45
  twistMotion=Free
 }
 {
  bone1Name=dyn_front_left_hair_01_mstr
  bone2Name=dyn_hair_01
  swing1LimitDegrees=45
  swing1Motion=Free
  swing2LimitDegrees=45
  swing2Motion=Free
  twistLimitDegrees=45
  twistMotion=Free
 }
 {bone1Name=dyn_front_left_hair_01, bone2Name=dyn_front_left_hair_01_mstr, swing1LimitDegrees=20, swing2LimitDegrees=25, twistLimitDegrees=5}
 {bone1Name=dyn_front__left_hair_01_b, bone2Name=dyn_front_left_hair_01, swing1LimitDegrees=15, swing2LimitDegrees=25, twistLimitDegrees=5}
 {bone1Name=dyn_front_left_hair_01_a, bone2Name=dyn_front_left_hair_01, swing1LimitDegrees=15, swing2LimitDegrees=25, twistLimitDegrees=5}
 {bone1Name=dyn_front_left_hair_01_c, bone2Name=dyn_front_left_hair_01, swing1LimitDegrees=15, swing2LimitDegrees=25, twistLimitDegrees=5}
 {
  bone1Name=dyn_hair_bang_01_mstr
  bone2Name=dyn_hair_01
  swing1LimitDegrees=45
  swing1Motion=Free
  swing2LimitDegrees=45
  swing2Motion=Free
  twistLimitDegrees=45
  twistMotion=Free
 }
 {bone1Name=dyn_hair_bang_01, bone2Name=dyn_hair_bang_01_mstr, swing1LimitDegrees=5, swing2LimitDegrees=5, twistLimitDegrees=5}
 {
  bone1Name=dyn_hair_front_right_01_mstr
  bone2Name=dyn_hair_01
  swing1LimitDegrees=45
  swing1Motion=Free
  swing2LimitDegrees=45
  swing2Motion=Free
  twistLimitDegrees=45
  twistMotion=Free
 }
 {bone1Name=dyn_hair_front_right_01, bone2Name=dyn_hair_front_right_01_mstr, swing1LimitDegrees=30, swing2LimitDegrees=25, twistLimitDegrees=5}
 {bone1Name=dyn_hair_front_right_01_m, bone2Name=dyn_hair_front_right_01, swing1LimitDegrees=30, swing2LimitDegrees=25, twistLimitDegrees=5}
 {bone1Name=dyn_hair_front_right_01_n, bone2Name=dyn_hair_front_right_01, swing1LimitDegrees=30, swing2LimitDegrees=25, twistLimitDegrees=5}
 {
  bone1Name=dyn_left_back_hair_01_mstr
  bone2Name=dyn_hair_01
  swing1LimitDegrees=45
  swing1Motion=Free
  swing2LimitDegrees=45
  swing2Motion=Free
  twistLimitDegrees=45
  twistMotion=Free
 }
 {bone1Name=dyn_left_back_hair_01, bone2Name=dyn_left_back_hair_01_mstr, swing1LimitDegrees=10, swing2LimitDegrees=25, twistLimitDegrees=5}
 {bone1Name=dyn_left_back_hair_01_e, bone2Name=dyn_left_back_hair_01, swing1LimitDegrees=10, swing2LimitDegrees=25, twistLimitDegrees=5}
 {bone1Name=dyn_left_back_hair_01_f, bone2Name=dyn_left_back_hair_01, swing1LimitDegrees=10, swing2LimitDegrees=25, twistLimitDegrees=5}
 {bone1Name=dyn_left_back_hair_01_g, bone2Name=dyn_left_back_hair_01, swing1LimitDegrees=10, swing2LimitDegrees=25, twistLimitDegrees=5}
 {
  bone1Name=dyn_left_hair_01_mstr
  bone2Name=dyn_hair_01
  swing1LimitDegrees=45
  swing1Motion=Free
  swing2LimitDegrees=45
  swing2Motion=Free
  twistLimitDegrees=45
  twistMotion=Free
 }
 {bone1Name=dyn_left_hair_02, bone2Name=dyn_left_hair_01_mstr, swing1LimitDegrees=20, swing2LimitDegrees=25, twistLimitDegrees=5}
 {bone1Name=dyn_left_hair_02_d, bone2Name=dyn_left_hair_02, swing1LimitDegrees=10, swing2LimitDegrees=25, twistLimitDegrees=5}
 {
  bone1Name=dyn_right_back_hair_01_mstr
  bone2Name=dyn_hair_01
  swing1LimitDegrees=45
  swing1Motion=Free
  swing2LimitDegrees=45
  swing2Motion=Free
  twistLimitDegrees=45
  twistMotion=Free
 }
 {bone1Name=dyn_right_back_hair_01, bone2Name=dyn_right_back_hair_01_mstr, swing1LimitDegrees=10, swing2LimitDegrees=25, twistLimitDegrees=5}
 {bone1Name=dyn_right_back_hair_01_h, bone2Name=dyn_right_back_hair_01, swing1LimitDegrees=10, swing2LimitDegrees=25, twistLimitDegrees=5}
 {bone1Name=dyn_right_back_hair_01_i, bone2Name=dyn_right_back_hair_01, swing1LimitDegrees=10, swing2LimitDegrees=25, twistLimitDegrees=5}
 {
  bone1Name=dyn_right_hair_01_mstr
  bone2Name=dyn_hair_01
  swing1LimitDegrees=45
  swing1Motion=Free
  swing2LimitDegrees=45
  swing2Motion=Free
  twistLimitDegrees=45
  twistMotion=Free
 }
 {bone1Name=dyn_right_hair_01, bone2Name=dyn_right_hair_01_mstr, swing1LimitDegrees=10, swing2LimitDegrees=25, twistLimitDegrees=5}
 {bone1Name=dyn_right_hair_01_j, bone2Name=dyn_right_hair_01, swing1LimitDegrees=10, swing2LimitDegrees=25, twistLimitDegrees=5}
 {bone1Name=dyn_right_hair_01_k, bone2Name=dyn_right_hair_01, swing1LimitDegrees=10, swing2LimitDegrees=25, twistLimitDegrees=5}
 {bone1Name=dyn_right_hair_01_l, bone2Name=dyn_right_hair_01, swing1LimitDegrees=10, swing2LimitDegrees=25, twistLimitDegrees=5}
 {
  bone1Name=dyn_hair_mstr
  bone2Name=head
  swing1LimitDegrees=45
  swing1Motion=Free
  swing2LimitDegrees=45
  swing2Motion=Free
  twistLimitDegrees=45
  twistMotion=Free
 }
 {
  bone1Name=dyn_right_back_hair_01_i
  bone2Name=dyn_right_hair_01_j
  swing1LimitDegrees=45
  swing1Motion=Free
  swing2LimitDegrees=45
  swing2Motion=Free
  twistLimitDegrees=45
  twistMotion=Free
 }
 {
  bone1Name=dyn_left_hair_02_d
  bone2Name=dyn_left_back_hair_01_e
  swing1LimitDegrees=45
  swing1Motion=Free
  swing2LimitDegrees=45
  swing2Motion=Free
  twistLimitDegrees=45
  twistMotion=Free
 }
 {
  bone1Name=dyn_right_back_hair_01_h
  bone2Name=dyn_left_back_hair_01_g
  swing1LimitDegrees=45
  swing1Motion=Free
  swing2LimitDegrees=45
  swing2Motion=Free
  twistLimitDegrees=45
  twistMotion=Free
 }
]
